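; Player gameplay ability for the equipped Chain Whip that manages the hookshot reticle aim component on the avatar pawn.
; On activation with authority, it adds the component to the Fort player pawn only if none exists, and stores the reference it created.
; On end, the authority side destroys that stored component before calling the parent end logic.
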(event Ability|EventActivateAbility
  ;[0.01964 0.125 0.03717] ActivateAbility: Apply the hookshot reticle aim pawn component to the avatar player.
  (|CallParentFunction)
  (Ability|HasAuthority
    (:True
      (bind _outfortplayerpawn (CallFunction|GetAvatarasValidFortPlayerPawn)
        (:ValidFortPlayerPawn
          (Utilities|IsValid (Actor|GetComponentbyClass _outfortplayerpawn "/SunRoseChainWhipGameplay/Items/ChainWhip/Abilities/Hookshot/B_ChainWhipHookshotReticleAimPawnComponent.B_ChainWhipHookshotReticleAimPawnComponent_C")
            (:Is%20Not%20Valid
              (bind _returnvalue (Game|AddComponentbyClass _outfortplayerpawn "/SunRoseChainWhipGameplay/Items/ChainWhip/Abilities/Hookshot/B_ChainWhipHookshotReticleAimPawnComponent.B_ChainWhipHookshotReticleAimPawnComponent_C"))
              (Variables|Transient|SetHookshotReticleAimPawnComponent _returnvalue))))))))

(event Ability|EventOnEndAbility
  (params (bWasCancelled :type Boolean))
  ;[0.1615 0.1143 0.01825] EndPlay: Remove the added pawn component from the avatar pawn.
  (Utilities|FlowControl|Sequence
    (:then_0
      (Ability|HasAuthority
        (:True
          (bind _hookshotreticleaimpawncomponent (Variables|Transient|GetHookshotReticleAimPawnComponent)
            (:then
              (Class|BAutoAimWeaponPawnComponentBase|DestroyComponent _hookshotreticleaimpawncomponent))
            (:else)))))
    (:then_1
      (|CallParentFunction bWasCancelled))))
